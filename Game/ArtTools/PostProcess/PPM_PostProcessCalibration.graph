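# Post-process material that displays calibration reference images on screen, with no scene color input.
# Viewport UVs are aspect-corrected around the center and multiplied by a Size parameter, then offset and scaled per element via ScaleUVsByCenter.
# A gamma chart blended with a horizontal grayscale ramp is clipped by two one-axis SphereMasks.
# A color checker and a ColorPostScreenshot texture are masked by their scaled-UV bounds and added.
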Material /Game/ArtTools/PostProcess/PPM_PostProcessCalibration
graph {
 properties {materialDomain=MD_PostProcess, pixelDepthOffsetMode=PDOM_Legacy, usedShadingModels=Default Lit}
 nodes [
  n1 = Constant() {r=0.5}
  n2 = Constant2Vector() {r=0.5}
  n3 = Subtract(A=ScreenPosition().ViewportUV, B=@n2)
  n4 = ViewProperty()
  n5 = AppendVector(A=Divide(A=ComponentMask(None=@n4.Property) {r=true}, B=ComponentMask(None=@n4.Property) {g=true}), B=Constant() {r=1})
  n6 = Multiply(A=Add(A=Multiply(A=@n3, B=@n5), B=@n2), B=ScalarParameter() {defaultValue=1, parameterName=Size})
  n7 = MaterialFunctionCall(Float2=@n6) {materialFunction=&BreakOutFloat2Components}
  n8 = SphereMask(A=@n1, B=@n7.R) {attenuationRadius=0.5}
  n9 = SphereMask(A=@n7.G, B=@n1) {attenuationRadius=0.5}
  n10 = Multiply(A=@n6) {constB=2}
  n11 = MaterialFunctionCall(
   UVs=@n10
   Texture Scale=ComponentMask(None=VectorParameter() {parameterName=ImageScale, defaultValue=<rgba 1 0.781 0 1>}.RGB) {g=true, r=true}
  ) {materialFunction=&ScaleUVsByCenter}
  n12 = Multiply(A=@n11.0-1 mask, B=TextureSample(UVs=@n11.UVs) {texture=&/GammaChart}.RGB)
  n13 = VectorParameter() {parameterName=GradientOffsetPosition, defaultValue=<rgba -0.0005 0.237 0 1>}
  n14 = MaterialFunctionCall(
   UVs=Add(A=@n10, B=ComponentMask(None=@n13.RGB) {g=true, r=true})
   Texture Scale=ComponentMask(None=VectorParameter() {parameterName=GradientScale, defaultValue=<rgba 0.9134 0.08 0 1>}.RGB) {g=true, r=true}
  ) {materialFunction=&ScaleUVsByCenter}
  n15 = LinearInterpolate(Alpha=ComponentMask(None=@n14.UVs) {r=true})
  n16 = MaterialFunctionCall(
   UVs=Subtract(A=@n10, B=Constant2Vector() {r=1})
   Texture Scale=ComponentMask(None=VectorParameter() {parameterName=ImageScale2, defaultValue=<rgba 1 0.781 0 1>}.RGB) {g=true, r=true}
  ) {materialFunction=&ScaleUVsByCenter}
  n17 = VectorParameter() {parameterName=ImageScale2b, defaultValue=<rgba 0.75 0.6 0 1>}
  n18 = TextureSample(UVs=Multiply(A=@n16.UVs, B=ComponentMask(None=@n17.RGB) {g=true, r=true})) {texture=/Engine/EditorMeshes/ColorCalibrator/Color_checker}
  n19 = MaterialFunctionCall(
   UVs=Subtract(A=@n10, B=Constant2Vector() {g=0.9, r=0.5})
   Texture Scale=ComponentMask(None=VectorParameter() {parameterName=ImageScale3, defaultValue=<rgba 1.567 0.942 0 1>}.RGB) {g=true, r=true}
  ) {materialFunction=&ScaleUVsByCenter}
  n20 = TextureSample(UVs=@n19.UVs) {texture=&/ColorPostScreenshot}
  Add(
   A=Multiply(A=Multiply(A=@n8, B=@n9), B=LinearInterpolate(A=@n12, B=@n15, Alpha=@n14.0-1 mask))
   B=Add(A=Multiply(A=@n16.0-1 mask, B=@n18.RGB), B=Multiply(A=@n19.0-1 mask, B=@n20.RGB))
  )
 ]
}
